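# Looping ambient sound cue for small clock ticks at the Helios train station POI, routed to the ambient props sound class.
# A loop node repeats a fixed half-second delay followed by a random pick between two equally weighted big-clock tick waves.
# Raised pitch and reduced volume repurpose the big-clock samples for a smaller clock, with overridden attenuation keeping the sound local.
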SoundCue /Game/Sounds/Helios/POI/TrainStation/Spots/SmallClockTicks_Loop_Cue
root=Looping_0
nodes [
 {id=Delay_0, children [Random_0], properties {delayMax=0.5, delayMin=0.5}}
 {id=Looping_0, children [Delay_0]}
 {id=Random_0, children [WavePlayer_3,WavePlayer_4], properties {weights [1,1]}}
 {id=WavePlayer_3, properties {soundWaveAssetPtr=/Game/Sounds/Fort_Prop_Sounds/Exterior/BigClock/BigClock_Tick_01}}
 {id=WavePlayer_4, properties {soundWaveAssetPtr=/Game/Sounds/Fort_Prop_Sounds/Exterior/BigClock/BigClock_Tick_02}}
]
properties {
 attenuationSettings=/Game/Sounds/Attenuation/Fort_BigClock_Att
 bOverrideAttenuation=true
 maxDistance=600
 pitchMultiplier=1.4
 soundClassObject=&Fort_Ambient_Props_Athena
 volumeMultiplier=0.351
 attenuationOverrides {falloffDistance=400, attenuationShapeExtents {x=200}}
}
