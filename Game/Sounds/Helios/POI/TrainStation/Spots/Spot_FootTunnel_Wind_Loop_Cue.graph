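# Looping ambient wind spot sound cue for the foot tunnel at the Helios Train Station POI.
# A looping wave player feeds a Modulator with fixed volume and default pitch variation, then an Enveloper whose volume curve fades in from silence.
# The cue plays quietly in the indoor ambient sound class, with bus sends disabled and prop concurrency assigned alongside a low-count override.
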SoundCue /Game/Sounds/Helios/POI/TrainStation/Spots/Spot_FootTunnel_Wind_Loop_Cue
root=Enveloper_1
nodes [
 {
  id=Enveloper_1
  children [Modulator_1]
  properties {pitchCurve {editorCurveData {keys [{time=0}]}}, volumeCurve {editorCurveData {keys [<at 0 0>,<at 1 1>]}}}
 }
 {id=Modulator_1, children [WavePlayer_0], properties {volumeMax=1, volumeMin=1}}
 {id=WavePlayer_0, properties {bLooping=true, soundWaveAssetPtr=&/Spot_FootTunnel_Wind_Loop}}
]
properties {
 bEnableBusSends=false
 maxDistance=2097152
 soundClassObject=&Fort_Ambient_Indoors_Athena
 volumeMultiplier=0.15
 concurrencyOverrides {maxCount=2}
 concurrencySet [/Game/Sounds/Concurrency_Rules/Athena_Prop]
 preEffectBusSends [{}]
}
